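# Base editor utility report for GART that flags assets referenced by the level named in LevelToCheck.
# A non-empty PathBlacklist limits the check to matching package paths and adds the matched path to the report message.
# CheckReferencersForLevel searches Asset Registry referencers for the level, collecting __ExternalActors__ packages when no direct reference exists.
# It follows those packages through soft references, in passes capped by a Do N, to catch references via external actors.
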EditorUtilityBlueprint /Game/NeverCook/EditorUtilities/GART/BR_Env/Reports/Report_MapReference_Base
graphs [
 CheckReferencersForLevel {nodes [
   n1 = CheckReferencersForLevel!() -> (then, In Asset Data:Asset Data Structure) {is=title K2Node_FunctionEntry}
   > n2 = Sequence!(execute) -> (then_0, then_1, then_2:Exec)
   > n3 = Get Referencers!(
    execute
    self:Asset Registry Interface=GetAssetRegistry()
    PackageName:Name=Break Asset Data(AssetData=@n1)
    ReferenceOptions:"Asset Registry Dependency Options Structure (by ref)"=Get ReferenceOptions() -> (ReferenceOptions:Asset Registry Dependency Options Structure)
   ) -> (then, OutReferencers:Array of Names, ReturnValue:bool) {is=title K2Node_CallFunction}
   n4 = Contains Item(TargetArray=@n3, ItemToFind=Get LevelToCheck() -> (LevelToCheck:Name=None))
   > n5 = Branch(Condition=@n4)
   > Set l_referenceFound(l_referenceFound:bool=@n4) -> (Output_Get:bool)
   n6 = For Each Loop with Break!(Exec:Exec=@n5.else, Array:Array of Names=@n3, Break:Exec) -> (LoopBody:Exec, Array Element:Name, Array Index:int, Completed:Exec)
   n7 = "To String (Name)"(InName=@n6)
   > Branch(Condition=Contains Item(
     TargetArray:Array of Strings=ParseIntoArray(SourceString=@n7, Delimiter=/)
     ItemToFind:"String (by ref)"=MakeLiteralString(Value=__ExternalActors__)
    ))
   > Add Unique(TargetArray:Array of Names=Get l_externalActors() -> (l_externalActors:Array of Names), NewItem:"Name (by ref)"=@n6)
   Do N!(Enter:Exec=[@n2.then_1, @n14.then], n:int=5, Reset:Exec) -> (Exit:Exec, Counter:int) {is=title K2Node_MacroInstance}
   > n8 = For Each Loop with Break(Array=Get l_externalActors(), Break=@n12.then)
   > n9 = Get Referencers(
    self=GetAssetRegistry()
    PackageName=@n8
    ReferenceOptions=Get ReferenceOptions_extActor() -> (ReferenceOptions_extActor:Asset Registry Dependency Options Structure)
   )
   n10 = Contains Item(TargetArray:Array of Names=@n9, ItemToFind:"Name (by ref)"=Get LevelToCheck() -> (LevelToCheck=None))
   > n11 = Branch(Condition=@n10)
   > n12 = Set l_referenceFound(l_referenceFound=@n10)
   n13 = For Each Loop(Exec=@n11.else, Array:Array of Names=@n9) -> (Array Element:Name)
   > n14 = Add Unique(TargetArray=Get l_externalActors(), NewItem=@n13)
   Return Node!(execute=@n2.then_2, Reference to Level Found:bool=Get l_referenceFound() -> (l_referenceFound:bool))
  ]}
 EventGraph {links [], nodes []}
 RunFilterLogicOnAssetData {nodes [
   RunFilterLogicOnAssetData()
   > n1 = Set l_blacklistNotEmpty(l_blacklistNotEmpty:bool=Is Not Empty!(TargetArray=Get PathBlacklist() -> (PathBlacklist:Array of Strings)) -> (ReturnValue:bool)) -> (Output_Get:bool)
   > n2 = Branch(Condition=@n1)
   > n3 = For Each Loop with Break(Array:Array of Strings=Get PathBlacklist(), Break=@n14.then) -> (Array Element:String)
   n4 = Get AssetData()
   n5 = StartsWith(SourceString="To String (Name)"(InName=Break Asset Data(AssetData=@n4).PackagePath), InPrefix=@n3)
   n6 = Branch(execute=@n3.Completed, Condition=@n5)
   n7 = CheckReferencersForLevel!(execute=[@n6.then, @n2.else], self, In Asset Data:Asset Data Structure=Get AssetData()) -> (then, Reference to Level Found:bool) {is=title K2Node_CallFunction}
   > n8 = Branch(Condition=Get l_blacklistNotEmpty() -> (l_blacklistNotEmpty:bool))
   n9 = NOT Boolean(A=@n7)
   n10 = "To String (Name)"(InName=Get LevelToCheck() -> (LevelToCheck=None))
   n11 = "To String (Name)"(InName=Get LevelToCheck() -> (LevelToCheck=None))
   n12 = Get l_pathMatch() -> (l_pathMatch:String)
   > Return Node!(
    execute
    Valid:bool=@n9
    ReportData:Struct GART Report Data Structure=Make Struct GART Report Data(
     DisplayTitle_20_#1=Asset Referencers
     DisplayDescShort_29_#2=Append(A="Referenced by ", B=@n10)
     DisplayDescLong_30_#3=Append!(A="content in the ", B=@n12, C:String=" path should not be referenced by ", D:String=@n11) -> (ReturnValue)
    )
   )
   n13 = "To String (Name)"(InName=Get LevelToCheck() -> (LevelToCheck=None))
   Return Node(
    execute=@n8.else
    Valid=@n9
    ReportData=Make Struct GART Report Data(
     DisplayTitle_20_#1=Asset Referencers
     DisplayDescShort_29_#2=Append!(A="Referenced by ", B=@n13) -> (ReturnValue)
     Severity_46_#4=NewEnumerator0
    )
   )
   Return Node(execute=@n6.else)
   Branch(execute=@n3.LoopBody, Condition=@n5)
   > n14 = Set l_pathMatch(l_pathMatch:String=@n3) -> (Output_Get:String)
  ]}
]
variables [
 LevelToCheck {type=string, default=None}
 PathBlacklist {type="array<string>", default []}
 ReferenceOptions {
  type=AssetRegistryDependencyOptions
  default {
   bIncludeEditorOnlyPackageReferences=true
   bIncludeGamePackageReferences=true
   bIncludeHardManagementReferences=true
   bIncludeHardPackageReferences=true
   bIncludeSearchableNames=false
   bIncludeSoftManagementReferences=false
   bIncludeSoftPackageReferences=false
  }
 }
 ReferenceOptions_extActor {
  type=AssetRegistryDependencyOptions
  default {
   bIncludeEditorOnlyPackageReferences=true
   bIncludeGamePackageReferences=true
   bIncludeHardManagementReferences=false
   bIncludeHardPackageReferences=false
   bIncludeSearchableNames=false
   bIncludeSoftManagementReferences=false
   bIncludeSoftPackageReferences=true
  }
 }
]
